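# Material function that outputs a noise-driven world position offset for the ChronoSwarmGrenade insect swarm effect.
# It sums broad main and finer detail layers, each panning soft noise over world XY and remapping RGB to signed XYZ displacement.
# Both layers scale by intensity and object scale, with detail masked by vertex red.
# Particle random and a time offset from a parameter plus weighted vertex alpha desynchronize the motion.
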MaterialFunction /ChronoSwarmGrenade/FX/MF_InsectSwarm_WPO
graph {nodes [
  n1 = ScalarParameter() {defaultValue=1, parameterName=WPO_Detail_Intensity}
  n2 = MaterialFunctionCall() {materialFunction=&ObjectScale}
  n3 = WorldPosition()
  n4 = Divide(A=Divide(A=ComponentMask(None=@n3.XYZ) {g=true, r=true}, B=ScalarParameter() {defaultValue=5, parameterName=WPO_Detail_Tiling})) {constB=3.5}
  n5 = Add(A=Add(A=Time(), B=NamedRerouteUsage()), B=ParticleRandom()) {constB=0}
  n6 = ScalarParameter() {defaultValue=10, parameterName=WPO_Detail_Panning}
  n7 = TextureSample(UVs=Panner(Coordinate=Divide(A=@n4) {constB=4}, Time=Multiply(A=@n5, B=@n6) {constB=2}) {speedX=0.1, speedY=-0.1}) {texture=/Game/Effects/Fort_Effects/Emotes/UltraLight/Texture/T_SoftNoise_BC7}
  MaterialExpressionAppend3Vector_0 = MaterialXAppend3Vector!(A=LinearInterpolate(Alpha=@n7.R) {constA=-1}, B=LinearInterpolate(Alpha=@n7.G) {constA=-1}, C=LinearInterpolate(Alpha=@n7.B) {constA=-1})
  n8 = Multiply(A=Multiply(A=Divide(A=@n1) {constB=3}, B=@n2.Scale XYZ), B=@MaterialExpressionAppend3Vector_0)
  n9 = VertexColor()
  n10 = WorldPosition()
  n11 = ScalarParameter() {defaultValue=0.05, parameterName=WPO_Main_Tiling}
  n12 = Divide(A=Multiply(A=ComponentMask(None=@n10.XYZ) {g=true, r=true}, B=Divide(A=@n11) {constB=20})) {constB=3.5}
  n13 = Add(A=Add(A=Time(), B=NamedRerouteUsage()), B=ParticleRandom()) {constB=0}
  n14 = ScalarParameter() {defaultValue=5, parameterName=WPO_Main_Panning}
  n15 = TextureSample(UVs=Panner(Coordinate=Divide(A=@n12) {constB=4}, Time=Multiply(A=@n13, B=@n14) {constB=2}) {speedX=0.1, speedY=-0.1}) {texture=&T_SoftNoise_BC7}
  MaterialExpressionAppend3Vector_1 = MaterialXAppend3Vector(A=LinearInterpolate(Alpha=@n15.R) {constA=-1}, B=LinearInterpolate(Alpha=@n15.G) {constA=-1}, C=LinearInterpolate(Alpha=@n15.B) {constA=-1})
  n16 = MaterialFunctionCall() {materialFunction=&ObjectScale}
  n17 = Multiply(A=ScalarParameter() {defaultValue=0.8, parameterName=WPO_Main_Intensity}, B=@n16.Scale XYZ)
  FunctionOutput(None=Add(A=Multiply(A=@n8, B=@n9.R), B=Multiply(A=@MaterialExpressionAppend3Vector_1, B=@n17)))
  Length(None=@n16.Scale XYZ) {is=class MaterialExpressionLength}
  Length(None=@n2.Scale XYZ)
  n18 = ScalarParameter() {defaultValue=1, parameterName=UseVertexColorAlpha}
  n19 = VertexColor()
  NamedRerouteDeclaration(None=Add(A=ScalarParameter() {parameterName=InitialTimeOffset}, B=Multiply(A=@n18, B=@n19.A)) {constB=0}) {name=RandomTimeOffset, nodeColor=<rgba 1 0.0636 0 1>}
 ]}
