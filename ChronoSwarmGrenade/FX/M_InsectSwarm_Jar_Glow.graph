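# Unlit additive material for the soft glow on the Chrono Swarm Grenade's insect jar, usable on Niagara sprites.
# A radial sphere mask shapes the glow, tinted fixed red or particle color by parameter, with particle alpha and depth fade softening it.
# Vertex offset combines a scale-aware camera-facing sprite and camera push, weighted by UseSpriteWPO, with the shared MF_InsectSwarm_WPO motion.
# The glare texture sample is unconnected.
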Material /ChronoSwarmGrenade/FX/M_InsectSwarm_Jar_Glow
graph {
 properties {bUsedWithNiagaraSprites=true, blendMode=BLEND_Additive, pixelDepthOffsetMode=PDOM_Legacy, shadingModel=MSM_Unlit, usedShadingModels=Unlit}
 nodes [
  n1 = WorldPosition()
  n2 = MaterialFunctionCall(Offset Amount=Constant() {r=2}, World Position=@n1.XYZ) {materialFunction=/Engine/Functions/Engine_MaterialFunctions02/WorldPositionOffset/CameraOffset}
  n3 = MaterialFunctionCall() {materialFunction=&ObjectScale}
  n4 = MaterialFunctionCall!(
   Center Location=@n1.XYZ
   0-1 UVs
   XY scale=Multiply(A=ScalarParameter() {defaultValue=25, parameterName=SpriteSize}, B=Length(None=@n3.Scale XYZ) {is=class MaterialExpressionLength})
   OverridePivot
   "Normals (Optional)"
  ) {materialFunction=&Sprite}
  Add(
   A=Multiply(A=Add(A=@n2.Camera Offset, B=@n4.World position offset), B=ScalarParameter() {defaultValue=1, parameterName=UseSpriteWPO})
   B=MaterialFunctionCall() {materialFunction=&/MF_InsectSwarm_WPO}.Result
  )
  n5 = ParticleColor()
  n6 = Constant() {r=0.5}
  n7 = SphereMask(A=TextureCoordinate(), B=@n6, Radius=@n6) {hardnessPercent=0}
  Multiply(
   A=LinearInterpolate(A=Constant3Vector() {constant=<rgba 0.4635 0 0.0048 1>}, B=@n5.RGB, Alpha=ScalarParameter() {parameterName=UseParticleColor})
   B=@n7
  )
  Multiply(A=@n5.A, B=DepthFade(Opacity=@n7) {fadeDistanceDefault=10})
  TextureSample() {texture=/Game/Effects/Fort_Effects/Textures/T_SUN_Glare_A}
 ]
}
