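# Unlit translucent Niagara sprite material that draws animated wasps for an insect swarm effect.
# An 8x8 flipbook, offset by particle random and slowed by dynamic parameter SlowMode, blends two baked wasp textures per particle.
# UseWPO enables sprite scaling plus swarm world position offset, and switches to vertex-alpha-driven UV flipping and texture choice.
# A near-camera distance fade adjusts brightness, desaturation and opacity, with time-of-day brightness applied.
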Material /ChronoSwarmGrenade/FX/M_InsectSwarm_Wasp_SubUV
graph {
 properties {
  bEnableResponsiveAA=true
  bUsedWithNiagaraSprites=true
  blendMode=BLEND_Translucent
  pixelDepthOffsetMode=PDOM_Legacy
  shadingModel=MSM_Unlit
  usedShadingModels=Unlit
 }
 nodes [
  n1 = Time()
  n2 = ParticleRandom()
  n3 = DynamicParameter() {defaultValue=<rgba 0 0 0 1>, paramNames [SlowMode, Param2, Param3, RandomTexture]}
  n4 = Constant() {r=8}
  n5 = TextureCoordinate()
  n6 = VertexColor()
  n7 = Round(None=Sine(None=@n6.A) {period=4})
  n8 = LinearInterpolate(A=Constant2Vector() {g=1, r=1}, B=Constant2Vector() {g=1, r=-1}, Alpha=@n7) {constAlpha=1}
  n9 = ScalarParameter() {parameterName=UseWPO}
  n10 = MaterialFunctionCall(
   "Animation  Phase (0-1)"=Divide(A=Add(A=@n1, B=@n2), B=LinearInterpolate(Alpha=@n3.SlowMode) {constA=1, constB=4})
   Number of Rows=@n4
   Number of Columns=@n4
   UVs=LinearInterpolate(A=@n5, B=Multiply(A=@n5, B=@n8), Alpha=@n9)
  ) {materialFunction=&FlipBook}
  n11 = TextureSample(UVs=@n10.UVs) {texture=&/T_AI_Wildlife_InsectSwarm_SingleWasp_Baked_V4_1}
  n12 = TextureSample(UVs=@n10.UVs) {texture=&/T_AI_Wildlife_InsectSwarm_SingleWasp_Baked_V4}
  n13 = LinearInterpolate(A=@n3.RandomTexture, B=@n7, Alpha=@n9)
  n14 = LinearInterpolate(A=@n11.A, B=@n12.A, Alpha=@n13)
  n15 = OneMinus(None=@n14)
  n16 = LinearInterpolate(A=@n11.RGB, B=@n12.RGB, Alpha=@n13)
  n17 = Multiply(A=Add(A=@n16, B=Multiply(A=@n16, B=@n15)), B=ScalarParameter() {defaultValue=0.7, parameterName=ColorBrightness})
  n18 = VectorParameter() {parameterName=TOD, defaultValue=<rgba 2 2 1 1>}
  n19 = MaterialFunctionCall(Value_Morning=@n18.R, Value_Day=@n18.R, Value_Evening=@n18.R, Value_Night=@n18.R) {materialFunction=&Scalar_Time_of_Day}
  n20 = LinearInterpolate(
   A=@n19.Result
   B=Multiply(A=@n19.Result, B=ScalarParameter() {defaultValue=1, parameterName=DistanceBrightnessScalar}) {constB=2}
   Alpha=NamedRerouteUsage()
  )
  n21 = NamedRerouteUsage()
  MaterialFunctionCall(Emissive=Desaturation(None=Multiply(A=@n17, B=@n20), Fraction=LinearInterpolate(Alpha=@n21))) {materialFunction=&MF_FX_EmissiveAdjustments}
  n22 = ScalarParameter() {defaultValue=4000, parameterName=DistanceBrightnessLerp}
  NamedRerouteDeclaration(None=Saturate(None=MaterialFunctionCall(Distance=@n22) {materialFunction=&Near_Camera_Fade}.Result)) {name=DistanceLerper, nodeColor=<rgba 1 0.375 0 1>}
  n23 = ScalarParameter() {defaultValue=0.8, parameterName=SpriteSize}
  n24 = MaterialFunctionCall!(
   Center Location=WorldPosition().XYZ
   0-1 UVs
   XY scale=LinearInterpolate(A=@n23, B=Multiply(A=@n23) {constB=2}, Alpha=Sine(None=@n6.A) {period=8})
   OverridePivot
   "Normals (Optional)"
  ) {materialFunction=&Sprite}
  n25 = MaterialFunctionCall() {materialFunction=&/MF_InsectSwarm_WPO}
  NamedRerouteDeclaration(None=Multiply(A=ScalarParameter() {parameterName=UseWPO}, B=Add(A=@n24.World position offset, B=@n25.Result))) {name=WPO, nodeColor=<rgba 0.7109 0 1 1>}
  n26 = ScalarParameter() {defaultValue=3, parameterName=Opacity}
  n27 = Multiply(A=@n26)
  n28 = NamedRerouteUsage()
  NamedRerouteDeclaration(None=Multiply(A=@n14, B=LinearInterpolate(A=@n26, B=@n27, Alpha=@n28))) {name=Opacity, nodeColor=<rgba 0.3359 1 0 1>}
  NamedRerouteUsage()
  NamedRerouteUsage()
 ]
}
