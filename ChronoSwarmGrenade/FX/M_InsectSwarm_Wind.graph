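# Translucent, unlit, two-sided material for the insect swarm wind effect of the Chrono Swarm Grenade, judging by its name and path.
# A stretched, diagonally panning dust texture supplies a blue-channel mask, scaled by an Opacity parameter.
# Inverted Near_Camera_Fade and FadeBasedOnViewAngle functions attenuate that mask by camera distance and viewing angle.
# A separate warm tan vector parameter provides a color, though material output connections are absent here.
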Material /ChronoSwarmGrenade/FX/M_InsectSwarm_Wind
graph {
 properties {blendMode=BLEND_Translucent, pixelDepthOffsetMode=PDOM_Legacy, shadingModel=MSM_Unlit, twoSided=true, usedShadingModels=Unlit}
 nodes [
  Constant()
  Constant() {r=1}
  n1 = Time()
  Multiply(A=@n1, B=ScalarParameter() {defaultValue=1})
  n2 = TextureCoordinate()
  n3 = Constant2Vector() {g=3, r=1}
  n4 = TextureSample(UVs=Panner(Coordinate=Multiply(A=@n2, B=@n3), Time=@n1) {speedX=1.5, speedY=2}) {texture=/Game/Global/Textures/Default/PointSplatter/T_Tiling_Dust}
  n5 = OneMinus(None=MaterialFunctionCall(Distance=ScalarParameter() {defaultValue=200, parameterName=NearCameraFadeDis}) {materialFunction=&Near_Camera_Fade}.Result)
  Multiply(
   A=Multiply(A=Multiply(A=@n4.B, B=@n5), B=ScalarParameter() {defaultValue=1, parameterName=Opacity}) {constB=3}
   B=MaterialFunctionCall("FadeSensitivity (optional)"=Constant() {r=4}) {materialFunction=&FadeBasedOnViewAngle}.Result
  )
  VectorParameter() {defaultValue=<rgba 1 0.6981 0.4503 1>}
 ]
}
