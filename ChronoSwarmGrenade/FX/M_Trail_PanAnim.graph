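# Translucent unlit, two-sided trail material for Niagara mesh particles in the ChronoSwarmGrenade effects.
# SplineThicken sets thickness from the Width parameter times a dynamic Width.
# A soft smoke texture pans along UV1, timed by dynamic PanAnim plus a random offset at per-particle speeds.
# Particle color tints emissive; opacity math combines smoke, particle alpha and a sphere mask whose UV power exponent rises over particle life, shifting the visible band.
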Material /ChronoSwarmGrenade/FX/M_Trail_PanAnim
graph {
 properties {
  bUsedWithNiagaraMeshParticles=true
  blendMode=BLEND_Translucent
  pixelDepthOffsetMode=PDOM_Legacy
  shadingModel=MSM_Unlit
  twoSided=true
  usedShadingModels=Unlit
 }
 nodes [
  n1 = Multiply(A=ScalarParameter() {defaultValue=50, parameterName=Width}, B=DynamicParameter() {paramNames [PanAnim,Width,Param3,Param4]}.Width)
  MaterialFunctionCall(WidthBase=@n1, WidthTip=@n1) {materialFunction=&SplineThicken}
  n2 = ParticleColor()
  n3 = MaterialFunctionCall(Float2=TextureCoordinate() {coordinateIndex=1}) {materialFunction=&BreakOutFloat2Components}
  n4 = Multiply(A=@n3.R, B=ScalarParameter() {defaultValue=0.5, parameterName=Param_1})
  n5 = ParticleRandom()
  n6 = DynamicParameter() {paramNames [PanAnim,Width,Param3,Param4]}
  n7 = ParticleRandom()
  n8 = Multiply(A=Multiply(A=LinearInterpolate(Alpha=@n7) {constA=1, constB=2}) {constB=0.4}) {constB=-1}
  n9 = Constant()
  n10 = TextureSample(UVs=Panner(Coordinate=AppendVector(A=@n4, B=@n3.G), Time=Add(A=@n5, B=@n6.PanAnim), Speed=AppendVector(A=@n8, B=@n9)) {speedX=-0.5}) {samplerType=SAMPLERTYPE_Grayscale, texture=&T_Smoke_Trail_Soft}
  MaterialFunctionCall(Emissive=Multiply(A=@n2.RGB, B=@n10.RGB)) {materialFunction=&MF_FX_EmissiveAdjustments}
  MaterialFunctionCall() {materialFunction=&DebugTimeSine}
  n11 = Power(
   Base=MaterialFunctionCall(Float2=TextureCoordinate() {coordinateIndex=1}) {materialFunction=&BreakOutFloat2Components}.R
   Exp=LinearInterpolate(Alpha=ParticleRelativeTime()) {constA=0.1, constB=10}
  ) {constExponent=0.5}
  n12 = Constant() {r=0.5}
  Multiply(A=Multiply(A=@n10.R, B=SphereMask(A=@n11, B=@n12, Radius=@n12) {hardnessPercent=0}), B=@n2.A)
  Time()
  Time()
 ]
}
