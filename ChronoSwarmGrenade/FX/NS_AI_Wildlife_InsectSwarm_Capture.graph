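# Niagara effect in the ChronoSwarmGrenade content that plays a brief flash, apparently when an AI wildlife insect swarm is captured.
# FN_Light bursts one short-lived particle driving a growing light and sprite with color and alpha curves.
# FN_sparks bursts one mesh that shrinks and fades while ramping to bright blue-white.
# A disabled GPU InsectSwarm emitter holds curl-noise sprites with pulsing attraction toward a sampled skeletal-mesh position.
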NiagaraSystem /ChronoSwarmGrenade/FX/NS_AI_Wildlife_InsectSwarm_Capture
dependencies {
 usedDataInterfaces [&NiagaraDataInterfaceColorCurve, &NiagaraDataInterfaceCurve]
 usedDynamicInputs []
 usedModules [&SystemState]
}
emitters {
 FN_Light {
  emitterUpdateScript [
   &EmitterState()
   &SpawnBurst_Instantaneous(Spawn Count=1, Use Spawn Probability=false)
  ]
  particleSpawnScript [/Niagara/Modules/Spawn/Initialization/InitializeParticle(
    Lifetime=0.2
    Mass=&UniformRangedFloat(Minimum=0.6, Maximum=1.2)
    Module.Write Color=true
    Module.Write Lifetime=true
    Module.Write Mass=false
    Module.Write Material Random=true
    Module.Write Position=false
    Module.Write Scale=false
    Module.Write SpriteRotation=true
    Module.Write SpriteSize=true
    Position=&SimulationPosition()
    Sprite Rotation=&UniformRangedFloat(Minimum=0, Maximum=360)
    Sprite Size=<250 250>
   )]
  particleUpdateScript [
   &ParticleState()
   &ScaleColor(
    Scale Alpha=&FloatFromCurve(FloatCurve=DataInterface() {propertyValues {Curve {keys [<at 0 1 RCIM_Cubic>,<at 1 0>]}}})
    Scale RGB=&MakeVectorFromLinearColorRGB(Color=&ColorFromCurve(ColorCurve=DataInterface() {propertyValues {
        AlphaCurve {keys [<at 0 1 RCIM_Cubic>]}
        BlueCurve {keys [<at 0.0029 1 RCIM_Cubic>,<at 0.4489 0.5649>]}
        GreenCurve {keys [<at 0.0029 0.7731 RCIM_Cubic>,<at 0.4489 0.6449>]}
        RedCurve {keys [<at 0.0029 0.4876 RCIM_Cubic>,<at 0.4489 1>]}
       }}))
    ScaleA=true
    ScaleRGB=true
   )
   &SolveForcesAndVelocity(Clamp Velocity=false, Limit Acceleration=false)
   &ScaleSpriteSize(Scale Factor=&Vector2DFromFloat(Value=&FloatFromCurve(FloatCurve=DataInterface() {propertyValues {Curve {keys [<at 0 0>,<at 1 1>]}}}, Scale Curve=0.8)))
   /SetVariables(Particles.Light_Size:NiagaraFloat=&FloatFromCurve(FloatCurve=DataInterface() {propertyValues {Curve {keys [<at 0 0.2>,<at 1 0.6>]}}}, Scale Curve=200)) {isSetParameters=true}
   /SetVariables(Particles.Sprite_Color:LinearColor=&ColorFromCurve(ColorCurve=DataInterface() {propertyValues {
       AlphaCurve {keys [<at 0.0087 1 RCIM_Cubic>,<at 0.9819 0>]}
       BlueCurve {keys [<at 0.017 4>,<at 0.9744 2 RCIM_Cubic>]}
       GreenCurve {keys [<at 0.017 3.1939>,<at 0.9744 1.0043 RCIM_Cubic>]}
       RedCurve {keys [<at 0.017 0.9666>,<at 0.9744 0.5222 RCIM_Cubic>]}
      }})) {isSetParameters=true}
  ]
  rendererClasses [&NiagaraLightRendererProperties, &NiagaraSpriteRendererProperties]
  renderers [
   {rendererClass=&NiagaraLightRendererProperties, rendererIndex=0}
   {rendererClass=&NiagaraSpriteRendererProperties, rendererIndex=1}
  ]
 }
 FN_sparks {
  emitterUpdateScript [
   &EmitterState()
   &SpawnBurst_Instantaneous(Spawn Count=1, Use Spawn Probability=false)
  ]
  particleSpawnScript [&InitializeParticle(
    Lifetime=0.2
    Mesh Scale=<0.1 0.1 0.1>
    Module.Write Color=true
    Module.Write Lifetime=true
    Module.Write Mass=true
    Module.Write Material Random=true
    Module.Write Position=false
    Module.Write Scale=true
    Module.Write SpriteRotation=false
    Module.Write SpriteSize=false
    Position=&SimulationPosition()
   )]
  particleUpdateScript [
   &ParticleState()
   &SolveForcesAndVelocity(Clamp Velocity=false, Limit Acceleration=false)
   &ScaleColor(
    Scale Alpha=&FloatFromCurve(FloatCurve=DataInterface() {propertyValues {Curve {keys [<at 0 1 RCIM_Cubic>,<at 0.5924 0.1926>,<at 1 0>]}}})
    Scale RGB=&MakeVectorFromLinearColorRGB(Color=&ColorFromCurve(ColorCurve=DataInterface() {propertyValues {
        AlphaCurve {keys [<at 0 1 RCIM_Cubic>]}
        BlueCurve {keys [<at 0.0065 0>,<at 0.8152 25 RCIM_Cubic>]}
        GreenCurve {keys [<at 0.0065 0>,<at 0.8152 17.0193 RCIM_Cubic>]}
        RedCurve {keys [<at 0.0065 0>,<at 0.8152 11.7699 RCIM_Cubic>]}
       }}))
    ScaleA=true
    ScaleRGB=true
   )
   &ScaleMeshSize(Scale Factor=&VectorFromFloat(Value=&FloatFromCurve(
      FloatCurve=DataInterface() {propertyValues {Curve {keys [{arriveTangent=-1.1291, interpMode=RCIM_Cubic, leaveTangent=-1.1292, tangentMode=RCTM_User, time=0, value=1}, <at 1 0>]}}}
      Scale Curve=1.5
     )))
  ]
  rendererClasses [&NiagaraMeshRendererProperties]
  renderers [{rendererClass=&NiagaraMeshRendererProperties, rendererIndex=0}]
 }
 InsectSwarm {
  bEnabled=false
  simTarget=GPUComputeSim
  emitterUpdateScript [
   &EmitterState()
   &SpawnBurst_Instantaneous(
    Spawn Count=&RandomRangeInt(Minimum=9, Maximum=15)
    Use Spawn Probability=false
    UseLoopCountLimit=false
   ) {module=SpawnBurst_Instantaneous001}
  ]
  particleSpawnScript [
   /Niagara/Modules/Spawn/Initialization/V2/InitializeParticle(
    Lifetime=&Multiply_Float(
     A=&RandomRangeFloat(Minimum=0.3, Maximum=0.33)
     B=&Lerp_Float(A=1, B=0.7, Alpha=0.5)
    )
    Mesh Scale Mode=::NewEnumerator1
    Mesh Uniform Scale Max=0.4
    Mesh Uniform Scale Min=0.1
    Sprite Size Mode=::NewEnumerator3
    Sprite UV Mode=::NewEnumerator1
    Uniform Sprite Size=100
    UsePositionOffset=false
    Write Material Random=true
   )
   /SetVariables(Particles.WaspSpriteSize:Vector2f=&Lerp_Vector2(A=<5 5>, B=<8 8>, Alpha=&RandomRangeFloat(Minimum=0.1, Maximum=0.5))) {isSetParameters=true}
   &ShapeLocation(MaskSpawnGroup=false, Override Seed=false, UseOwnerScale=true)
  ]
  particleUpdateScript [
   &ParticleState()
   /SetVariables(Particles.WaspSpriteSize:Vector2f=&Multiply_Vector2D(
     A=$Particles.Initial.WaspSpriteSize
     B=&Multiply_Vector2D_ByFloat(
      Float=&FloatFromCurve(
       FloatCurve=DataInterface() {propertyValues {Curve {keys [
           <at 0 0 RCIM_Cubic>
           {arriveTangent=0.2423, interpMode=RCIM_Cubic, leaveTangent=0.2423, tangentMode=RCTM_User, time=0.2, value=1}
           {arriveTangent=0.0664, interpMode=RCIM_Cubic, leaveTangent=0.0664, tangentMode=RCTM_User, time=0.8, value=1}
           <at 1 0 RCIM_Cubic>
          ]}}}
       Scale Curve=&Lerp_Float(A=1, B=2, Alpha=0.5)
      )
      Vector2D=&Multiply_Vector2D_ByFloat(
       Float=&Lerp_Float(A=&RandomRangeFloat(Minimum=1.5, Maximum=2), B=4, Alpha=0.5)
       Vector2D=<1 1>
      )
     )
    )) {isSetParameters=true}
   &CurlNoiseForce(Noise Frequency=200, Noise Strength=5000, Pan Noise=true, Pan Noise Field=<1 1 1>)
   &PointAttractionForce(
    Attraction Radius=300
    Attraction Strength=&Multiply_Float(
     A=&RandomRangeFloat(Minimum=280, Maximum=300)
     B=&Multiply_Float(
      A=&Multiply_Float(
       A=&Lerp_Float(
        A=0.2
        B=0.6
        Alpha=&Waveform("Waveform [1]"=::NewEnumerator9, "[1] Frequency"=0.4, "[1] Amplitude Min/Max"=<0 1>)
       )
       B=3
      )
      B=1
     )
    )
    Attractor Position:NiagaraPosition=$Particles.SkeletalMeshLocation.SampledPosition
    Falloff Exponent=0.6
    Kill Within Radius=false
    Module.Use Falloff=true
    UseAttractorOffset=true
   )
   &ScaleSpriteSize(Uniform Curve Sprite Scale=DataInterface() {propertyValues {Curve {keys [{interpMode=RCIM_Cubic, tangentMode=RCTM_User, time=0, value=1}, {interpMode=RCIM_Cubic, tangentMode=RCTM_User, time=1, value=0}]}}})
   &ScaleColor(Scale Alpha=0, Scale RGB=<0.3 0 0>, ScaleA=true, ScaleRGB=true)
   &Drag(Drag=5, Use Linear Drag=true, Use Rotational Drag=true)
   &SolveForcesAndVelocity(Clamp Velocity=false, Limit Acceleration=false)
  ]
  rendererClasses [&NiagaraSpriteRendererProperties]
  renderers [
   {rendererClass=&NiagaraSpriteRendererProperties}
   {rendererClass=&NiagaraSpriteRendererProperties, rendererIndex=1}
  ]
 }
}
summary {
 systemName=NS_AI_Wildlife_InsectSwarm_Capture
 emitters [
  {emitterName=FN_Light, rendererClasses [&NiagaraLightRendererProperties, &NiagaraSpriteRendererProperties]}
  {emitterName=FN_sparks, rendererClasses [&NiagaraMeshRendererProperties]}
  {bEnabled=false, emitterName=InsectSwarm, simTarget=GPUComputeSim, rendererClasses [&NiagaraSpriteRendererProperties]}
 ]
 userVariables []
}
