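# Translucent, surface-lit material for mud on the Chrono Swarm Grenade's jar, judging by its name and textures.
# It multiplies the mud diffuse texture's alpha by an Opacity scalar parameter, so material instances can scale its transparency.
# A normal map supplies surface detail, alongside two scalar constants.
# Lighting uses Default Lit with per-pixel translucency lighting.
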Material /ChronoSwarmGrenade/Materials/M_Swarm_Jar_Mud
graph {
 properties {
  blendMode=BLEND_Translucent
  pixelDepthOffsetMode=PDOM_Legacy
  translucencyLightingMode=TLM_SurfacePerPixelLighting
  usedShadingModels=Default Lit
 }
 nodes [
  Constant()
  Constant() {r=0.8}
  Multiply(
   A=TextureSample() {texture=/ChronoSwarmGrenade/Textures/T_Swarm_Jar_Mud_D}.A
   B=ScalarParameter() {defaultValue=1, desc=Opacity Multiplier, parameterName=Opacity}
  )
  TextureSample() {samplerType=SAMPLERTYPE_Normal, texture=/ChronoSwarmGrenade/Textures/T_Swarm_Jar_Mud_N}
 ]
}
