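# Lit material for the wasp body mesh in the Chrono Swarm Grenade, enabled for use on Niagara mesh particles.
# It samples wasp diffuse, normal and mask textures for surface shading.
# It also calls the shared MF_InsectSwarm_WPO function, named for insect swarm World Position Offset, a vertex-shader technique for moving mesh vertices.
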Material /ChronoSwarmGrenade/Mesh/Materials/M_Wasp_Body
graph {
 properties {bUsedWithNiagaraMeshParticles=true, pixelDepthOffsetMode=PDOM_Legacy, usedShadingModels=Default Lit}
 nodes [
  MaterialFunctionCall() {materialFunction=/ChronoSwarmGrenade/FX/MF_InsectSwarm_WPO}
  TextureSample() {samplerType=SAMPLERTYPE_Masks, texture=/ChronoSwarmGrenade/Mesh/Textures/T_Wasp_Wasp_S}
  TextureSample() {samplerType=SAMPLERTYPE_Normal, texture=/ChronoSwarmGrenade/Mesh/Textures/T_Wasp_Wasp_N}
  TextureSample() {texture=/ChronoSwarmGrenade/Mesh/Textures/T_Wasp_Wasp_D}
 ]
}
