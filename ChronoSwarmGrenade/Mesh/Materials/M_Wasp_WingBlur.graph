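# Translucent, unlit, two-sided wasp wing material for Niagara mesh particles in the ChronoSwarm grenade, faking motion-blurred flapping.
# A particle-randomized sine rotates blur and sharp wing textures over different arcs, with the SlowMode dynamic parameter reducing flap speed.
# Combined textures plus view-angle-faded noise specular drive emissive, while blended alphas, gradient masks and FakeWingBlurOpacity shape opacity.
# MF_InsectSwarm_WPO supplies the swarm's vertex offset function.
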Material /ChronoSwarmGrenade/Mesh/Materials/M_Wasp_WingBlur
graph {
 properties {
  bUsedWithNiagaraMeshParticles=true
  blendMode=BLEND_Translucent
  pixelDepthOffsetMode=PDOM_Legacy
  shadingModel=MSM_Unlit
  twoSided=true
  usedShadingModels=Unlit
 }
 nodes [
  n1 = TextureCoordinate()
  n2 = Constant2Vector() {g=1}
  n3 = Add(A=Time(), B=ParticleRandom())
  n4 = LinearInterpolate(
   A=ScalarParameter() {defaultValue=20, parameterName=FlapSpeed}
   Alpha=DynamicParameter() {defaultValue=<rgba 0 1 1 1>, paramNames [SlowMode, FakeWingBlurOpacity, Param3, Param4]}.SlowMode
  ) {constA=20, constB=5.5}
  n5 = ConstantBiasScale(None=Sine(None=Multiply(A=@n3, B=@n4)))
  n6 = MaterialFunctionCall(UVs=@n1, Rotation Center=@n2, "Rotation Angle (0-1)"=LinearInterpolate(Alpha=@n5) {constA=-0.055, constB=0.06}) {materialFunction=&CustomRotator}
  n7 = TextureSample(UVs=@n6.Rotated Values) {samplerSource=SSM_Clamp_WorldGroupSettings, texture=/ChronoSwarmGrenade/Mesh/Textures/T_Wasp_Wasp_Wing_D_Blur}
  n8 = TextureSample(UVs=MaterialFunctionCall(UVs=@n1, Rotation Center=@n2, "Rotation Angle (0-1)"=LinearInterpolate(Alpha=@n5) {constA=-0.09, constB=0.12}) {materialFunction=&CustomRotator}.Rotated Values) {samplerSource=SSM_Clamp_WorldGroupSettings, texture=/ChronoSwarmGrenade/Mesh/Textures/T_Wasp_Wasp_Wing_D}
  n9 = Add(A=@n7.RGB, B=@n8.RGB)
  n10 = Multiply(
   A=TextureSample(UVs=@n6.Rotated Values) {samplerSource=SSM_Clamp_WorldGroupSettings, texture=/ChronoSwarmGrenade/Mesh/Textures/T_Wasp_Wasp_Wing_D_Blur_Noise}.R
   B=VectorParameter() {parameterName=WingSpecularEmissive, defaultValue=<rgba 60 30.9872 16.2 1>}.RGB
  )
  n11 = MaterialFunctionCall("FadeSensitivity (optional)"=Constant() {r=12}) {materialFunction=&FadeBasedOnViewAngle}
  MaterialFunctionCall(Emissive=Add(A=Multiply(A=@n9, B=@n8.RGB), B=Multiply(A=@n10, B=@n11.Result))) {materialFunction=&MF_FX_EmissiveAdjustments}
  MaterialFunctionCall() {materialFunction=/ChronoSwarmGrenade/FX/MF_InsectSwarm_WPO}
  n12 = Power(Base=SphereMask(A=TextureCoordinate(), B=Constant2Vector() {g=1}) {attenuationRadius=1, hardnessPercent=0})
  n13 = Add(
   A=Divide(A=@n7.A, B=ScalarParameter() {defaultValue=6, parameterName=FakeWingBlurOpacity}) {constB=5}
   B=Multiply(A=@n8.A, B=LinearInterpolate(Alpha=@n12) {constA=0.01})
  )
  n14 = DynamicParameter() {defaultValue=<rgba 0 1 1 1>, paramNames [SlowMode, FakeWingBlurOpacity, Param3, Param4]}
  n15 = MaterialFunctionCall() {materialFunction=&LinearGradient}
  n16 = Multiply(A=Multiply(A=OneMinus(None=@n15.VGradient))) {constB=8}
  Multiply(
   A=Multiply(A=Multiply(A=@n13, B=@n14.FakeWingBlurOpacity), B=Saturate(None=@n16))
   B=ScalarParameter() {defaultValue=0.22, parameterName=OverallOpacity}
  )
 ]
}
